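# Physics asset for the dynamic hood, necklace and tag bones of the HeavyRoar medium male player body cosmetic.
# Capsules form left and right hood-trap chains and a three-link necklace ending in two tags, all rooted at spine_05.
# Chain links have limited swing and very tight twist, while hood master, upper-arm and spine joints rotate freely.
# Larger spine and upper-arm capsules supply body volumes around the dangling parts.
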PhysicsAsset /BRCosmetics/Characters/Player/Male/Medium/Bodies/M_MED_HeavyRoar/Meshes/M_MED_HeavyRoar_Dyn_Hood
bodies {
 dyn_hoodTraps_01_l [{shapeType=Capsule, length=4.75, radius=2, shapeName=dyn_hoodTraps_01_l_capsule, center=<6 0 0.5>, rotation=<rot 90 0 0>}]
 dyn_hoodTraps_01_r [{shapeType=Capsule, length=4.25, radius=2, shapeName=dyn_hoodTraps_01_r_capsule, center=<6.5843 0 0>, rotation=<rot 90 0 0>}]
 dyn_hoodTraps_02_l [{shapeType=Capsule, length=4.75, radius=2, shapeName=dyn_hoodTraps_02_l_capsule, center=<6 0 0>, rotation=<rot 90 0 0>}]
 dyn_hoodTraps_02_r [{shapeType=Capsule, length=4.75, radius=2, shapeName=dyn_hoodTraps_02_r_capsule, center=<6 0 0>, rotation=<rot 90 0 0>}]
 dyn_hoodTraps_l_MSTR [{
   shapeType=Capsule
   length=0.7039
   radius=0.6317
   shapeName=dyn_hoodTraps_l_MSTR_capsule
   center=<-0.3482 0.1201 -0.1083>
   rotation=<rot 0 -87.9178 -80.7651>
  }]
 dyn_hoodTraps_r_MSTR [{
   shapeType=Capsule
   length=0.505
   radius=0.505
   shapeName=dyn_hoodTraps_r_MSTR_capsule
   center=<-0.2483 -0.0052 -0.2903>
   rotation=<rot 0 68.15 -64.6069>
  }]
 dyn_necklace_01 [{shapeType=Capsule, length=3.0716, radius=3.9284, shapeName=dyn_necklace_01_capsule, center=<1 0 5>, rotation=<rot 82.419 -180 -180>}]
 dyn_necklace_02 [{shapeType=Capsule, length=2.7356, radius=4.2644, shapeName=dyn_necklace_02_capsule, center=<1 0 5>, rotation=<rot 90 0 0>}]
 dyn_necklace_03 [{shapeType=Capsule, length=2.3216, radius=4.6784, shapeName=dyn_necklace_03_capsule, center=<1 0 4>, rotation=<rot 86.5235 -180 -180>}]
 dyn_tag_01 [{shapeType=Capsule, length=1.875, radius=2.125, shapeName=dyn_tag_01_capsule, center=<2 0 1>, rotation=<rot 90 0 0>}]
 dyn_tag_02 [{shapeType=Capsule, length=1, radius=3, shapeName=dyn_tag_02_capsule, center=<2 -1 2.3>, rotation=<rot 90 0 0>}]
 spine_04 [{shapeType=Capsule, length=16.4761, radius=15.3905, shapeName=spine_05_capsule, center=<1.7252 -3.0205 0>}]
 spine_05 [
  {shapeType=Capsule, length=17.4761, radius=14.3905, shapeName=spine_05_capsule, center=<-4.7659 -4.5669 0>}
  {shapeType=Capsule, length=3, radius=15, shapeName=spine_05_capsule, center=<-3.6899 -2.262 0>}
 ]
 upperarm_l [{shapeType=Capsule, length=17.75, radius=6, shapeName=upperarm_r_capsule, center=<11.4176 -2.5885 0.6771>, rotation=<rot 90 180 -9.1689>}]
 upperarm_r [{
   shapeType=Capsule
   length=17.75
   radius=6
   shapeName=upperarm_r_capsule
   center=<-11.3316 1.4094 -0.5447>
   rotation=<rot 90 -166.5268 -170.696>
  }]
}
constraints [
 {bone1Name=dyn_hoodTraps_01_l, bone2Name=dyn_hoodTraps_l_MSTR, swing1LimitDegrees=60, swing2LimitDegrees=45, twistLimitDegrees=2}
 {bone1Name=dyn_hoodTraps_02_l, bone2Name=dyn_hoodTraps_01_l, swing1LimitDegrees=60, swing2LimitDegrees=45, twistLimitDegrees=2}
 {bone1Name=dyn_hoodTraps_01_r, bone2Name=dyn_hoodTraps_r_MSTR, swing1LimitDegrees=60, swing2LimitDegrees=45, twistLimitDegrees=2}
 {bone1Name=dyn_hoodTraps_02_r, bone2Name=dyn_hoodTraps_01_r, swing1LimitDegrees=60, swing2LimitDegrees=45, twistLimitDegrees=2}
 {
  bone1Name=dyn_hoodTraps_l_MSTR
  bone2Name=spine_05
  swing1LimitDegrees=45
  swing1Motion=Free
  swing2LimitDegrees=45
  swing2Motion=Free
  twistLimitDegrees=45
  twistMotion=Free
 }
 {
  bone1Name=dyn_hoodTraps_r_MSTR
  bone2Name=spine_05
  swing1LimitDegrees=45
  swing1Motion=Free
  swing2LimitDegrees=45
  swing2Motion=Free
  twistLimitDegrees=45
  twistMotion=Free
 }
 {
  bone1Name=upperarm_r
  bone2Name=spine_05
  swing1LimitDegrees=45
  swing1Motion=Free
  swing2LimitDegrees=45
  swing2Motion=Free
  twistLimitDegrees=45
  twistMotion=Free
 }
 {
  bone1Name=upperarm_l
  bone2Name=spine_05
  swing1LimitDegrees=45
  swing1Motion=Free
  swing2LimitDegrees=45
  swing2Motion=Free
  twistLimitDegrees=45
  twistMotion=Free
 }
 {bone1Name=dyn_necklace_01, bone2Name=spine_05, swing1LimitDegrees=7, swing2LimitDegrees=30, twistLimitDegrees=2}
 {bone1Name=dyn_necklace_02, bone2Name=dyn_necklace_01, swing1LimitDegrees=7, swing2LimitDegrees=30, twistLimitDegrees=2}
 {bone1Name=dyn_necklace_03, bone2Name=dyn_necklace_02, swing1LimitDegrees=7, swing2LimitDegrees=30, twistLimitDegrees=2}
 {bone1Name=dyn_tag_01, bone2Name=dyn_necklace_03, swing1LimitDegrees=30, swing2LimitDegrees=5, twistLimitDegrees=2}
 {bone1Name=dyn_tag_02, bone2Name=dyn_necklace_03, swing1LimitDegrees=30, swing2LimitDegrees=2, twistLimitDegrees=2}
 {
  bone1Name=spine_05
  bone2Name=spine_04
  swing1LimitDegrees=45
  swing1Motion=Free
  swing2LimitDegrees=45
  swing2Motion=Free
  twistLimitDegrees=45
  twistMotion=Free
 }
]
