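# Physics asset for the dynamic jacket of the HeavyRoar medium male player body.
# Seven capsule chains (back, sides, front) hang from MSTR root bodies on spine_02, with narrow twist and asymmetric swing limits, plus a zipper off the right front.
# Cross constraints tie the center back tip to both neighboring back tips, and the left front tip to the left side tip.
# Pelvis, spine, groin and thigh capsules supply body volumes.
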PhysicsAsset /BRCosmetics/Characters/Player/Male/Medium/Bodies/M_MED_HeavyRoar/Meshes/M_MED_HeavyRoar_Dyn_Jacket
bodies {
 dyn_groin_collision [{shapeType=Capsule, length=16.4052, radius=11.5756, center=<-3.4702 -1.2035 -0.0072>, rotation=<rot 0 0 -0.2739>}]
 dyn_jacket_bk_01 [{shapeType=Capsule, length=19.4006, radius=10.5994, shapeName=dyn_jacket_l_bk_01_capsule, center=<8 0 10>, rotation=<rot 0 0 90>}]
 dyn_jacket_bk_02 [{shapeType=Capsule, length=19.1506, radius=10.8494, shapeName=dyn_jacket_l_bk_02_capsule, center=<8 0 10>, rotation=<rot 0 0 90>}]
 dyn_jacket_bk_MSTR [{shapeType=Capsule, length=1, radius=1}]
 dyn_jacket_l_01 [{shapeType=Capsule, length=11.6203, radius=10, shapeName=dyn_jacket_l_01_capsule, center=<8 0 9>, rotation=<rot 0 0 90>}]
 dyn_jacket_l_02 [{shapeType=Capsule, length=14.3303, radius=10, shapeName=dyn_jacket_l_02_capsule, center=<8 0 9>, rotation=<rot 0 0 89.7261>}]
 dyn_jacket_l_MSTR [{shapeType=Capsule, length=1, radius=1}]
 dyn_jacket_l_bk_01 [{shapeType=Capsule, length=14.25, radius=10.75, shapeName=dyn_jacket_l_bk_01_capsule, center=<10 0 8>, rotation=<rot -0.0001 0.0001 80>}]
 dyn_jacket_l_bk_02 [{shapeType=Capsule, length=14.75, radius=10.25, shapeName=dyn_jacket_l_bk_02_capsule, center=<9 0 8>, rotation=<rot -0.0001 0.0001 80>}]
 dyn_jacket_l_bk_MSTR [{shapeType=Capsule, length=1, radius=1}]
 dyn_jacket_l_fr_01 [{shapeType=Capsule, length=5.3157, radius=9.0876, shapeName=dyn_jacket_l_fr_01_capsule, center=<6 1 8>, rotation=<rot 0 0 115>}]
 dyn_jacket_l_fr_02 [{shapeType=Capsule, length=6.6187, radius=9.8191, shapeName=dyn_jacket_l_fr_02_capsule, center=<6 1 10>, rotation=<rot 0.0002 0.0002 115>}]
 dyn_jacket_l_fr_03 [{shapeType=Capsule, length=5.8687, radius=10.5691, shapeName=dyn_jacket_l_fr_02_capsule, center=<6 1 10>, rotation=<rot 0.0002 0.0002 115>}]
 dyn_jacket_l_fr_MSTR [{shapeType=Capsule, length=1, radius=1}]
 dyn_jacket_r_01 [{shapeType=Capsule, length=11.6203, radius=10, shapeName=dyn_jacket_l_01_capsule, center=<8 0 9>, rotation=<rot 0 0 90>}]
 dyn_jacket_r_02 [{shapeType=Capsule, length=14.3303, radius=10, shapeName=dyn_jacket_l_02_capsule, center=<8 0 9>, rotation=<rot 0 0 89.7261>}]
 dyn_jacket_r_MSTR [{shapeType=Capsule, length=1, radius=1}]
 dyn_jacket_r_bk_01 [{shapeType=Capsule, length=14.25, radius=10.75, shapeName=dyn_jacket_l_bk_01_capsule, center=<10 0 8>, rotation=<rot -0.0001 -0.0001 100>}]
 dyn_jacket_r_bk_02 [{shapeType=Capsule, length=14.75, radius=10.25, shapeName=dyn_jacket_l_bk_02_capsule, center=<9 0 8>, rotation=<rot -0.0001 -0.0001 100>}]
 dyn_jacket_r_bk_MSTR [{shapeType=Capsule, length=1, radius=1}]
 dyn_jacket_r_fr_01 [{
   shapeType=Capsule
   length=3.5657
   radius=9.0876
   shapeName=dyn_jacket_l_fr_01_capsule
   center=<6 -3 8>
   rotation=<rot 90 -177.9294 -178.2033>
  }]
 dyn_jacket_r_fr_02 [{shapeType=Capsule, length=6.9937, radius=10.5691, shapeName=dyn_jacket_l_fr_02_capsule, center=<6 0 10>, rotation=<rot 0.0001 -0.0001 80>}]
 dyn_jacket_r_fr_03 [{shapeType=Capsule, length=6.6187, radius=10.9441, shapeName=dyn_jacket_l_fr_02_capsule, center=<6 0 10>, rotation=<rot 0.0001 -0.0001 80>}]
 dyn_jacket_r_fr_MSTR [{shapeType=Capsule, length=1, radius=1}]
 dyn_zipper [{shapeType=Capsule, length=2.75, radius=2.5, shapeName=dyn_zipper_capsule, center=<2.681 0 2>, rotation=<rot 90 0 0>}]
 pelvis [{shapeType=Capsule, length=8.0735, radius=14.1573, center=<-2.7926 0.0312 0>}]
 spine_01 [{shapeType=Capsule, length=5.0021, radius=14.2268, center=<-3.2898 0.5438 0>}]
 spine_02 [{shapeType=Capsule, length=6.5021, radius=13.2268, center=<-3.2898 -0.3263 0>}]
 thigh_twist_01_l [{shapeType=Capsule, length=29.3195, radius=12.4125, center=<3.6027 0.177 1.2704>, rotation=<rot 83.653 178.1802 -4.3304>}]
 thigh_twist_01_r [{shapeType=Capsule, length=29.3195, radius=12.4125, center=<-3.6771 0.6429 -1.2696>, rotation=<rot 86.5757 176.6478 176.6794>}]
}
constraints [
 {bone1Name=dyn_jacket_r_bk_01, bone2Name=dyn_jacket_r_bk_MSTR, swing1LimitDegrees=8, swing2LimitDegrees=45, twistLimitDegrees=2}
 {bone1Name=dyn_jacket_r_bk_02, bone2Name=dyn_jacket_r_bk_01, swing1LimitDegrees=8, swing2LimitDegrees=45, twistLimitDegrees=2}
 {bone1Name=dyn_jacket_l_bk_01, bone2Name=dyn_jacket_l_bk_MSTR, swing1LimitDegrees=8, swing2LimitDegrees=45, twistLimitDegrees=2}
 {bone1Name=dyn_jacket_l_bk_02, bone2Name=dyn_jacket_l_bk_01, swing1LimitDegrees=8, swing2LimitDegrees=45, twistLimitDegrees=2}
 {bone1Name=dyn_jacket_l_fr_01, bone2Name=dyn_jacket_l_fr_MSTR, swing1LimitDegrees=10, swing2LimitDegrees=45, twistLimitDegrees=2}
 {bone1Name=dyn_jacket_l_fr_02, bone2Name=dyn_jacket_l_fr_01, swing1LimitDegrees=10, swing2LimitDegrees=45, twistLimitDegrees=2}
 {bone1Name=dyn_jacket_r_fr_01, bone2Name=dyn_jacket_r_fr_MSTR, swing1LimitDegrees=5, swing2LimitDegrees=45, twistLimitDegrees=2}
 {bone1Name=dyn_jacket_r_fr_02, bone2Name=dyn_jacket_r_fr_01, swing1LimitDegrees=5, swing2LimitDegrees=45, twistLimitDegrees=2}
 {bone1Name=dyn_jacket_l_02, bone2Name=dyn_jacket_l_01, swing1LimitDegrees=7, swing2LimitDegrees=45, twistLimitDegrees=2}
 {bone1Name=dyn_jacket_l_01, bone2Name=dyn_jacket_l_MSTR, swing1LimitDegrees=7, swing2LimitDegrees=45, twistLimitDegrees=2}
 {bone1Name=dyn_jacket_r_01, bone2Name=dyn_jacket_r_MSTR, swing1LimitDegrees=7, swing2LimitDegrees=45, twistLimitDegrees=2}
 {bone1Name=dyn_jacket_r_02, bone2Name=dyn_jacket_r_01, swing1LimitDegrees=7, swing2LimitDegrees=45, twistLimitDegrees=2}
 {
  bone1Name=dyn_jacket_r_bk_MSTR
  bone2Name=spine_02
  swing1LimitDegrees=45
  swing1Motion=Free
  swing2LimitDegrees=45
  swing2Motion=Free
  twistLimitDegrees=45
  twistMotion=Free
 }
 {
  bone1Name=dyn_jacket_l_bk_MSTR
  bone2Name=spine_02
  swing1LimitDegrees=45
  swing1Motion=Free
  swing2LimitDegrees=45
  swing2Motion=Free
  twistLimitDegrees=45
  twistMotion=Free
 }
 {
  bone1Name=dyn_jacket_bk_MSTR
  bone2Name=spine_02
  swing1LimitDegrees=45
  swing1Motion=Free
  swing2LimitDegrees=45
  swing2Motion=Free
  twistLimitDegrees=45
  twistMotion=Free
 }
 {
  bone1Name=dyn_jacket_r_MSTR
  bone2Name=spine_02
  swing1LimitDegrees=45
  swing1Motion=Free
  swing2LimitDegrees=45
  swing2Motion=Free
  twistLimitDegrees=45
  twistMotion=Free
 }
 {
  bone1Name=dyn_jacket_l_MSTR
  bone2Name=spine_02
  swing1LimitDegrees=45
  swing1Motion=Free
  swing2LimitDegrees=45
  swing2Motion=Free
  twistLimitDegrees=45
  twistMotion=Free
 }
 {
  bone1Name=dyn_jacket_l_fr_MSTR
  bone2Name=spine_02
  swing1LimitDegrees=45
  swing1Motion=Free
  swing2LimitDegrees=45
  swing2Motion=Free
  twistLimitDegrees=45
  twistMotion=Free
 }
 {
  bone1Name=dyn_jacket_r_fr_MSTR
  bone2Name=spine_02
  swing1LimitDegrees=45
  swing1Motion=Free
  swing2LimitDegrees=45
  swing2Motion=Free
  twistLimitDegrees=45
  twistMotion=Free
 }
 {
  bone1Name=thigh_twist_01_l
  bone2Name=pelvis
  swing1LimitDegrees=45
  swing1Motion=Free
  swing2LimitDegrees=45
  swing2Motion=Free
  twistLimitDegrees=45
  twistMotion=Free
 }
 {bone1Name=dyn_jacket_bk_01, bone2Name=dyn_jacket_bk_MSTR, swing1LimitDegrees=8, swing2LimitDegrees=45, twistLimitDegrees=2}
 {bone1Name=dyn_jacket_bk_02, bone2Name=dyn_jacket_bk_01, swing1LimitDegrees=8, swing2LimitDegrees=45, twistLimitDegrees=2}
 {
  bone1Name=spine_01
  bone2Name=pelvis
  swing1LimitDegrees=45
  swing1Motion=Free
  swing2LimitDegrees=45
  swing2Motion=Free
  twistLimitDegrees=45
  twistMotion=Free
 }
 {
  bone1Name=dyn_groin_collision
  bone2Name=pelvis
  swing1LimitDegrees=45
  swing1Motion=Free
  swing2LimitDegrees=45
  swing2Motion=Free
  twistLimitDegrees=45
  twistMotion=Free
 }
 {
  bone1Name=dyn_jacket_bk_02
  bone2Name=dyn_jacket_l_bk_02
  swing1LimitDegrees=45
  swing1Motion=Free
  swing2LimitDegrees=20
  twistLimitDegrees=45
  twistMotion=Free
 }
 {
  bone1Name=dyn_jacket_bk_02
  bone2Name=dyn_jacket_r_bk_02
  swing1LimitDegrees=45
  swing1Motion=Free
  swing2LimitDegrees=20
  twistLimitDegrees=45
  twistMotion=Free
 }
 {
  bone1Name=thigh_twist_01_r
  bone2Name=pelvis
  swing1LimitDegrees=45
  swing1Motion=Free
  swing2LimitDegrees=45
  swing2Motion=Free
  twistLimitDegrees=45
  twistMotion=Free
 }
 {bone1Name=dyn_jacket_l_fr_03, bone2Name=dyn_jacket_l_fr_02, swing1LimitDegrees=10, swing2LimitDegrees=45, twistLimitDegrees=2}
 {bone1Name=dyn_jacket_r_fr_03, bone2Name=dyn_jacket_r_fr_02, swing1LimitDegrees=5, swing2LimitDegrees=45, twistLimitDegrees=2}
 {bone1Name=dyn_zipper, bone2Name=dyn_jacket_r_fr_03, swing1LimitDegrees=45, swing2LimitDegrees=45, twistLimitDegrees=5}
 {
  bone1Name=spine_02
  bone2Name=spine_01
  swing1LimitDegrees=45
  swing1Motion=Free
  swing2LimitDegrees=45
  swing2Motion=Free
  twistLimitDegrees=45
  twistMotion=Free
 }
 {
  bone1Name=dyn_jacket_l_fr_03
  bone2Name=dyn_jacket_l_02
  swing1LimitDegrees=45
  swing1Motion=Free
  swing2LimitDegrees=45
  swing2Motion=Free
  twistLimitDegrees=45
  twistMotion=Free
 }
]
