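# Physics asset for the dynamic necklace and tags on the HeavyRoar medium male body, defining capsule bodies for secondary motion.
# Three necklace links chain down from spine_05, with two tags branching from the last link under tight twist and asymmetric swing limits.
# Larger capsules on spine_04 and spine_05 stand in for the torso, joined by a constraint with free swing and twist.
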PhysicsAsset /BRCosmetics/Characters/Player/Male/Medium/Bodies/M_MED_HeavyRoar/Meshes/M_MED_HeavyRoar_Dyn_Necklace
bodies {
 dyn_necklace_01 [{shapeType=Capsule, length=3.0716, radius=3.9284, shapeName=dyn_necklace_01_capsule, center=<1 0 5>, rotation=<rot 82.419 -180 -180>}]
 dyn_necklace_02 [{shapeType=Capsule, length=2.7356, radius=4.2644, shapeName=dyn_necklace_02_capsule, center=<1 0 5>, rotation=<rot 90 0 0>}]
 dyn_necklace_03 [{shapeType=Capsule, length=2.3216, radius=4.6784, shapeName=dyn_necklace_03_capsule, center=<1 0 4>, rotation=<rot 86.5235 -180 -180>}]
 dyn_tag_01 [{shapeType=Capsule, length=1.75, radius=2.25, shapeName=dyn_tag_01_capsule, center=<2 0 1>, rotation=<rot 90 0 0>}]
 dyn_tag_02 [{shapeType=Capsule, length=2, radius=3, shapeName=dyn_tag_02_capsule, center=<2 -1 2.3>, rotation=<rot 90 0 0>}]
 spine_04 [{shapeType=Capsule, length=16.4761, radius=15.3905, shapeName=spine_05_capsule, center=<1.7252 -3.0205 0>}]
 spine_05 [
  {shapeType=Capsule, length=17.4761, radius=14.3905, shapeName=spine_05_capsule, center=<-4.7659 -4.5669 0>}
  {shapeType=Capsule, length=3, radius=15, shapeName=spine_05_capsule, center=<-3.6899 -2.262 0>}
 ]
}
constraints [
 {bone1Name=dyn_necklace_01, bone2Name=spine_05, swing1LimitDegrees=7, swing2LimitDegrees=30, twistLimitDegrees=2}
 {bone1Name=dyn_necklace_02, bone2Name=dyn_necklace_01, swing1LimitDegrees=7, swing2LimitDegrees=30, twistLimitDegrees=2}
 {bone1Name=dyn_necklace_03, bone2Name=dyn_necklace_02, swing1LimitDegrees=7, swing2LimitDegrees=30, twistLimitDegrees=2}
 {bone1Name=dyn_tag_01, bone2Name=dyn_necklace_03, swing1LimitDegrees=30, swing2LimitDegrees=5, twistLimitDegrees=2}
 {bone1Name=dyn_tag_02, bone2Name=dyn_necklace_03, swing1LimitDegrees=30, swing2LimitDegrees=2, twistLimitDegrees=2}
 {
  bone1Name=spine_05
  bone2Name=spine_04
  swing1LimitDegrees=45
  swing1Motion=Free
  swing2LimitDegrees=45
  swing2Motion=Free
  twistLimitDegrees=45
  twistMotion=Free
 }
]
