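# Character-part AnimBP for the HeavyRoar body, adding secondary motion to jacket, hood and necklace over the pose copied from the parent mesh.
# A jacket Control Rig feeds hood-root AnimDynamics and RigidBody sims, with a necklace look-at fallback when rigid bodies are disabled.
# Material variant, backbling, skydiving and emote state toggle hood and necklace dynamics; skydive-aware gravity drives the jacket.
# The built-in emote curve blends slot animation onto hood bones.
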AnimBlueprint /BRCosmetics/Characters/Player/Male/Medium/Bodies/M_MED_HeavyRoar/Meshes/M_MED_HeavyRoar_AnimBP
target_skeleton=/Game/Characters/Player/Male/Medium/Bodies/M_MED_HeavyRoar/Meshes/M_MED_HeavyRoar_Skeleton
graphs [
 AnimGraph {nodes [
   "Anim Dynamics - Chain: dyn_hood_01 -> dyn_hood_02"!(ComponentPose, bAlphaBoolEnabled:bool=True, ExternalForce="0,0,0") -> (Pose) {
    alphaInputType=Bool
    angularDampingOverride=0.9999
    bChain=true
    bGravityOverrideInSimSpace=true
    bOverrideAngularDamping=true
    bUseGravityOverride=true
    simulationSpace=BoneRelative
    alphaBoolBlend {blendOption=Linear}
    boundBone {boneName=dyn_hood_01}
    chainEnd {boneName=dyn_hood_02}
    componentAppliedLinearAccClamp=<3000 3000 3000>
    componentLinearAccScale=<0.2 0.2 0.2>
    componentLinearVelScale=<0.3 0.3 0.3>
    gravityOverride=<-980 -100>
    physicsBodyDefinitions [
     {
      boundBone {boneName=dyn_hood_01}
      boxExtents=<10 15 8>
      constraintSetup {angularLimitsMax=<0 5 10>, angularLimitsMin=<0 -35 -10>, angularTarget=<0 0 0>}
      localJointOffset=<1.5 0 0>
     }
     {
      boundBone {boneName=dyn_hood_02}
      boxExtents=<10 18 8>
      constraintSetup {angularLimitsMax=<0 5 10>, angularLimitsMin=<0 -35 -10>, angularTarget=<0 0 0>}
      localJointOffset=<1.5 0 -1>
     }
    ]
    relativeSpaceBone {boneName=spine_04}
   }
   Component To Local()
   n1 = "Constraint - dyn_jacket_bk_02 (P)"(Alpha=1, ConstraintWeights_0=1) {
    alphaBoolBlend {blendOption=Linear}
    boneToModify {boneName=dyn_jacket_bk_02}
    constraintSetup [{targetBone {boneName=pelvis}}]
    constraintWeights [1]
   }
   n2 = "Constraint - dyn_jacket_l_bk_02 (P)"(ComponentPose=@n1, Alpha=1, ConstraintWeights_0=1) {
    alphaBoolBlend {blendOption=Linear}
    boneToModify {boneName=dyn_jacket_l_bk_02}
    constraintSetup [{targetBone {boneName=pelvis}}]
    constraintWeights [1]
   }
   n3 = "Constraint - dyn_jacket_r_bk_02 (P)"(ComponentPose=@n2, Alpha=1, ConstraintWeights_0=1) {
    alphaBoolBlend {blendOption=Linear}
    boneToModify {boneName=dyn_jacket_r_bk_02}
    constraintSetup [{targetBone {boneName=pelvis}}]
    constraintWeights [1]
   }
   n4 = "Constraint - dyn_jacket_l_02 (P)"(ComponentPose=@n3, Alpha=1, ConstraintWeights_0=1) {
    alphaBoolBlend {blendOption=Linear}
    boneToModify {boneName=dyn_jacket_l_02}
    constraintSetup [{targetBone {boneName=pelvis}}]
    constraintWeights [1]
   }
   "Constraint - dyn_jacket_r_02 (P)"(ComponentPose=@n4, Alpha=1, ConstraintWeights_0=1) {
    alphaBoolBlend {blendOption=Linear}
    boneToModify {boneName=dyn_jacket_r_02}
    constraintSetup [{targetBone {boneName=pelvis}}]
    constraintWeights [1]
   }
   Control Rig() {
    alphaBoolBlend {blendOption=Linear}
    controlRigAssetReference {blueprintRigClass=/Game/Characters/Player/Male/Medium/Bodies/M_MED_HeavyRoar/Meshes/M_MED_HeavyRoar_DeformRig_ControlRig.M_MED_HeavyRoar_DeformRig_ControlRig_C}
   }
   n5 = Control Rig(Source=Use cached pose 'prePose'()) {
    alphaBoolBlend {blendOption=Linear}
    controlRigAssetReference {blueprintRigClass=&/M_MED_HeavyRoar_Jacket_CtrlRig.M_MED_HeavyRoar_Jacket_CtrlRig_C}
   }
   n6 = Anim Dynamics - Bone: dyn_hood_root!(ComponentPose=Local To Component(LocalPose=@n5), bAlphaBoolEnabled:bool=Get hoodDynAlpha() -> (hoodDynAlpha:bool), ExternalForce="0,0,0") -> (Pose) {
    alphaInputType=Bool
    angularDampingOverride=0.9999
    bGravityOverrideInSimSpace=true
    bOverrideAngularDamping=true
    bUseGravityOverride=true
    simulationSpace=BoneRelative
    alphaBoolBlend {blendOption=Linear}
    boundBone {boneName=dyn_hood_root}
    componentAppliedLinearAccClamp=<3000 3000 3000>
    componentLinearAccScale=<0.15 0.15 0.15>
    componentLinearVelScale=<0.3 0.3 0.3>
    gravityOverride=<-980 100>
    physicsBodyDefinitions [{
      boundBone {boneName=dyn_hood_root}
      boxExtents=<10 15 8>
      constraintSetup {angularLimitsMax=<0 3 10>, angularLimitsMin=<0 -20 -10>, angularTarget=<0 0 0>}
      localJointOffset=<0 0 -1.5>
     }]
    relativeSpaceBone {boneName=spine_04}
   }
   n7 = RigidBody!(ComponentPose=@n6, Alpha, OverrideWorldGravity:Vector=Get GravityOverride(), ExternalForce) -> (Pose) {
    bClampLinearTranslationLimitToRefPose=true
    bOverrideWorldGravity=true
    overridePhysicsAsset=&/M_MED_HeavyRoar_Dyn_Jacket
    simulationSpace=BaseBoneSpace
    alphaBoolBlend {blendOption=Linear}
    baseBoneRef {boneName=spine_02}
    componentAppliedLinearAccClamp=<3000 1000 1000>
    componentLinearAccScale {x=0.15}
    componentLinearVelScale {x=0.2}
    overrideWorldGravity {z=-980}
    simSpaceSettings {maxAngularAcceleration=100, maxAngularVelocity=2, maxLinearAcceleration=500, maxLinearVelocity=100, velocityScaleZ=0, worldAlpha=1}
   }
   n8 = Get isWearHood() -> (isWearHood:bool)
   n9 = RigidBody!(ComponentPose=@n7, bAlphaBoolEnabled:bool=NOT Boolean(A=@n8), OverrideWorldGravity="0,0,-980", ExternalForce) -> (Pose) {
    alphaInputType=Bool
    bClampLinearTranslationLimitToRefPose=true
    bOverrideWorldGravity=true
    overridePhysicsAsset=&/M_MED_HeavyRoar_Dyn_Hood
    simulationSpace=BaseBoneSpace
    alphaBoolBlend {blendOption=Linear}
    baseBoneRef {boneName=spine_05}
    componentAppliedLinearAccClamp=<3000 1000 1000>
    componentLinearAccScale {x=0.15}
    componentLinearVelScale {x=0.2}
    overrideWorldGravity {z=-980}
    simSpaceSettings {maxAngularAcceleration=100, maxAngularVelocity=2, maxLinearAcceleration=500, maxLinearVelocity=100, velocityScaleZ=0, worldAlpha=1}
   }
   n10 = RigidBody(ComponentPose=@n9, bAlphaBoolEnabled=@n8, OverrideWorldGravity="0,0,-980") {
    alphaInputType=Bool
    bClampLinearTranslationLimitToRefPose=true
    bOverrideWorldGravity=true
    overridePhysicsAsset=&/M_MED_HeavyRoar_Dyn_Necklace
    simulationSpace=BaseBoneSpace
    alphaBoolBlend {blendOption=Linear}
    baseBoneRef {boneName=spine_05}
    componentAppliedLinearAccClamp=<3000 1000 1000>
    componentLinearAccScale {x=0.15}
    componentLinearVelScale {x=0.2}
    overrideWorldGravity {z=-980}
    simSpaceSettings {maxAngularAcceleration=100, maxAngularVelocity=2, maxLinearAcceleration=500, maxLinearVelocity=100, velocityScaleZ=0, worldAlpha=1}
   }
   n11 = Look At - Bone: dyn_necklace_02(ComponentPose=Local To Component(LocalPose=Use cached pose 'prePose'()), Alpha=1) {
    alphaBoolBlend {blendOption=Linear}
    boneToModify {boneName=dyn_necklace_02}
    lookAtLocation=<0 15.4012 0>
    lookAtTarget {boneReference {boneName=spine_04}}
    lookAt_Axis {axis=<1 0>}
   }
   n12 = Layered blend per bone(
    BasePose=Blend Poses by bool(BlendPose_0=Component To Local(ComponentPose=@n10), BlendPose_1=Component To Local(ComponentPose=@n11), bActiveValue=Get bRigidBodyEnabled()) {blendType=Linear}
    BlendPoses_0=Slot 'DefaultSlot'()
    BlendWeights_0=Get isEmoteHeavyRoar() -> (isEmoteHeavyRoar:double)
   ) {layerSetup [{branchFilters [{boneName=anim_hood_bk}, {boneName=anim_hood_bt}, {boneName=anim_hood_l}, {boneName=anim_hood_r}]}]}
   Output Pose(Result=@n12) {layerGroup=None}
   Save cached pose 'prePose'(Pose=Copy Pose From Mesh() {bCopyCurves=true, bUseAttachedParent=true}) {cacheName=prePose}
  ]}
 EventGraph {nodes [
   Event BlueprintUpdateAnimation()
   > n1 = Sequence!(execute) -> (then_0, then_1, then_2:Exec, then_3:Exec)
   > n2 = Gravity Override State Function DEPRECATED(CharacterPartAnimInstance=Self-Reference(), GameTimeInSeconds=GetGameTimeInSeconds(), GravityOverrideParams=Get Gravity Override Params())
   > Set GravityOverride(GravityOverride=@n2)
   Set isWearBackbling(execute=@n1.then_1, isWearBackbling:bool=CheckForBackbling()) -> (Output_Get:bool)
   > n3 = Cast To FortPlayerPawn(Object=GetOwningActor())
   > n4 = GetVariantChannelActiveIndexForItemActor!(
    execute
    Item:Athena Cosmetic Item Definition Object Reference=/BRCosmetics/Athena/Items/Cosmetics/Characters/Character_HeavyRoar
    Actor:Actor Object Reference=@n3
    Channel:Gameplay Tag Structure=(TagName="Cosmetics.Variant.Channel.Material")
    VariantList:Array of Gameplay Tag Structures=Make Array("[0]"=(TagName="Cosmetics.Variant.Property.Mat3"))
    bErrorIfChannelMissing:bool=false
    bErrorIfTagNotInArray:bool=false
   ) -> (then, Index:int=0, bIndexIsValid:bool, bFoundChannel:bool)
   > n5 = Branch(Condition=@n4.bIndexIsValid)
   > Set isWearHood(isWearHood:bool=true) -> (Output_Get:bool)
   Set isWearHood(execute=@n5.else, isWearHood=false)
   n6 = Branch(execute=@n1.then_2, Condition=Get bIsInFrontEnd())
   > Set isEmoteHeavyRoar(isEmoteHeavyRoar:double=GetCurveValue(self=GetOwnerFrontEndAnimInstance(), CurveName=emote_BuiltIn)) -> (Output_Get:double)
   Set isEmoteHeavyRoar(
    execute=@n6.else
    isEmoteHeavyRoar=GetCurveValue(
     self=GetOwnerFortPlayerAnimInstance!(self:Custom Character Part Anim Instance Object Reference) -> (ReturnValue:Fort Player Anim Instance Object Reference)
     CurveName=emote_BuiltIn
    )
   )
   n7 = Branch(
    execute=@n1.then_3
    Condition=OR Boolean!(
     A=Get isWearBackbling() -> (isWearBackbling:bool)
     B=Get isWearHood()
     C:bool=Get bIsSkyDiving()
     D:bool=Get isEmotingHeavyRoar() -> (isEmotingHeavyRoar:bool)
    ) -> (ReturnValue)
   )
   > Set hoodDynAlpha(hoodDynAlpha:bool=false) -> (Output_Get:bool)
   Set hoodDynAlpha(execute=@n7.else, hoodDynAlpha=true)
   n8 = GetAnimInstance(self=GetOwnerSkeletalMeshComponent())
   Is Valid(InputObject=@n8)
   > Set isEmotingHeavyRoar(isEmotingHeavyRoar:bool=Montage_IsPlaying(self=@n8, Montage=/BRCosmetics/Animation/Game/MainPlayer/Emotes/HeavyRoar/Emote_HeavyRoar_CMM_M)) -> (Output_Get:bool)
  ]}
]
variables [
 Gravity Override Params {
  type=GravityOverrideParamsStruct
  default {
   skydive_FloatingDown {gravityOverride=<0 -700 -100>, windAmplitude=<200 -500 200>, windOffset=<0 -600 100>}
   skydive_GoingUp {gravityOverride=<0 -300 500>, windAmplitude=<200 -500 300>}
   skydive_Parachuting {gravityOverride=<0 -500 -980>, windAmplitude=<200 -150 400>, windFrequency=<1.5 4 6>}
  }
 }
 GravityOverride {type=Vector}
 hoodDynAlpha {type=boolean}
 isEmoteHeavyRoar {type=number}
 isEmotingHeavyRoar {type=boolean}
 isWearBackbling {type=boolean}
 isWearHood {type=boolean}
]
