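# Control Rig for the hood part of the male medium HeavyRoar player body, correcting hood bones as the head, shoulders and right arm move.
# Two spherical pose readers on the head, relative to spine_05, weight corrective transforms on hood deformer bones.
# Weighted parent constraints then blend both hood trapezius bones between spine, clavicle and, on the left, upper-arm twist.
# A final pose reader on the right upper arm adds an offset to the right trapezius bone only.
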ControlRigBlueprint /BRCosmetics/Characters/Player/Male/Medium/Bodies/M_MED_HeavyRoar/Meshes/Parts/M_MED_HeavyRoar_Hood_CtrlRig
graphs [
 RigVMFunctionLibrary {links [], nodes []}
 RigVMModel {nodes [
   RigUnit_BeginExecution = &RigUnit_BeginExecution()
   n1 = &RigUnit_SphericalPoseReader(
    ActiveRegionScaleFactors=(PositiveWidth=0.6,NegativeWidth=0.6,PositiveHeight=1,NegativeHeight=1)
    ActiveRegionSize=0.1
    Debug=(bDrawDebug=false,DebugScale=25,DebugSegments=20,DebugThickness=0.25)
    DriverAxis=(X=1,Y=0,Z=0)
    DriverItem=(Type=Bone,Name="head")
    FalloffRegionScaleFactors=(PositiveWidth=0.7,NegativeWidth=0.7,PositiveHeight=1,NegativeHeight=1)
    FalloffSize=0.3
    OptionalParentItem=(Type=Bone,Name="spine_05")
    RotationOffset=(X=0,Y=-90,Z=-60)
    ExecutePin=@RigUnit_BeginExecution.ExecutePin
   )
   > &RigUnit_ModifyTransforms(
    ItemToModify="((Item=(Type=Bone,Name=\"def_hood_bk_MSTR\"),Transform=(Rotation=(X=0,Y=0,Z=0,W=1),Translation=(X=0,Y=0,Z=0),Scale3D=(X=0.8,Y=1,Z=1))),(Item=(Type=Bone,Name=\"def_hood_r_MSTR\"),Transform=(Rotation=(X=0,Y=0,Z=0,W=1),Translation=(X=0,Y=0,Z=0),Scale3D=(X=0,Y=1,Z ...(truncated)"
    WeightMinimum=0
    ItemToModify.0.Transform.Rotation=&RigVMFunction_MathQuaternionFromEuler(Euler=(X=0,Y=-50,Z=0)).Result
    ItemToModify.1.Transform.Rotation=&RigVMFunction_MathQuaternionFromEuler(Euler=(X=0,Y=0,Z=20)).Result
    ItemToModify.2.Transform.Rotation=&RigVMFunction_MathQuaternionFromEuler(Euler=(X=0,Y=0,Z=30)).Result
    ItemToModify.3.Transform.Rotation=&RigVMFunction_MathQuaternionFromEuler(Euler=(X=0,Y=0,Z=-30)).Result
    ItemToModify.4.Transform.Rotation=&RigVMFunction_MathQuaternionFromEuler(Euler=(X=0,Y=0,Z=-20)).Result
    Weight=@n1.OutputParam
   )
   > n2 = &RigUnit_SphericalPoseReader(
    ActiveRegionScaleFactors=(PositiveWidth=0.7,NegativeWidth=0.7,PositiveHeight=1,NegativeHeight=1)
    ActiveRegionSize=0.1
    Debug=(bDrawDebug=false,DebugScale=25,DebugSegments=20,DebugThickness=0.25)
    DriverAxis=(X=1,Y=0,Z=0)
    DriverItem=(Type=Bone,Name="head")
    FalloffRegionScaleFactors=(PositiveWidth=0.7,NegativeWidth=0.7,PositiveHeight=1,NegativeHeight=1)
    FalloffSize=0.3
    OptionalParentItem=(Type=Bone,Name="spine_05")
    RotationOffset=(X=-30,Y=35,Z=90)
   )
   > &RigUnit_ModifyTransforms(
    ItemToModify="((Item=(Type=Bone,Name=\"def_hood_bk_r_MSTR\"),Transform=(Rotation=(X=0,Y=0,Z=0,W=1),Translation=(X=-2,Y=0,Z=3),Scale3D=(X=0.8,Y=1,Z=1))),(Item=(Type=Bone,Name=\"def_hood_bk_MSTR\"),Transform=(Rotation=(X=0,Y=0,Z=0,W=1),Translation=(X=0,Y=0,Z=0),Scale3D=(X=0.6 ...(truncated)"
    ItemToModify.0.Transform.Rotation=&RigVMFunction_MathQuaternionFromEuler(Euler=(X=0,Y=-50,Z=15)).Result
    ItemToModify.1.Transform.Rotation=&RigVMFunction_MathQuaternionFromEuler(Euler=(X=0,Y=-4,Z=-15)).Result
    ItemToModify.2.Transform.Rotation=&RigVMFunction_MathQuaternionFromEuler(Euler=(X=0,Y=-30,Z=40)).Result
    Weight=@n2.OutputParam
   )
   > &RigUnit_ParentConstraint(
    AdvancedSettings=(RotationOrderForFilter=XZY)
    Child=(Type=Bone,Name="dyn_hoodTraps_r_MSTR")
    Parents=((Item=(Type=Bone,Name="spine_05"),Weight=0.45),(Item=(Type=Bone,Name="clavicle_r"),Weight=0.55))
    Weight=1
   )
   > &RigUnit_ParentConstraint(
    AdvancedSettings=(RotationOrderForFilter=XZY)
    Child=(Type=Bone,Name="dyn_hoodTraps_l_MSTR")
    Parents=((Item=(Type=Bone,Name="spine_04"),Weight=0.1823),(Item=(Type=Bone,Name="spine_05"),Weight=0.304),(Item=(Type=Bone,Name="clavicle_l"),Weight=0.4107),(Item=(Type=Bone,Name="upperarm_twist_01_l"),Weight=0.1031))
    Weight=1
   )
   > n3 = &RigUnit_SphericalPoseReader(
    ActiveRegionScaleFactors=(PositiveWidth=0.4,NegativeWidth=0.8,PositiveHeight=0.8,NegativeHeight=0.8)
    ActiveRegionSize=0.15
    Debug=(bDrawDebug=false,bDraw2D=false,bDrawLocalAxes=true,DebugScale=25,DebugSegments=20,DebugThickness=0.25)
    DriverAxis=(X=-1,Y=0,Z=0)
    DriverItem=(Type=Bone,Name="upperarm_r")
    FalloffRegionScaleFactors=(PositiveWidth=0.6,NegativeWidth=0.6,PositiveHeight=1,NegativeHeight=1)
    FalloffSize=0.4
    FlipHeightScaling=false
    FlipWidthScaling=true
    OptionalParentItem=(Type=Bone,Name="spine_04")
    RotationOffset=(X=-115,Y=30,Z=20)
   )
   > &RigUnit_ModifyTransforms(
    ItemToModify=((Item=(Type=Bone,Name="dyn_hoodTraps_r_MSTR"),Transform=(Rotation=(X=0,Y=0,Z=0,W=1),Translation=(X=-1,Y=-4,Z=3),Scale3D=(X=1,Y=1,Z=1))))
    ItemToModify.0.Transform.Rotation=&RigVMFunction_MathQuaternionFromEuler(Euler=(X=0,Y=0,Z=-30)).Result
    Weight=@n3.OutputParam
   )
  ]}
]
hierarchy [
 root {loc=<0 0 0>, rot=<0 0 0 1>}
 pelvis {parent=root, loc=<0 -0.7071 93.1191>, rot=<0.0127 0.707 -0.0127 -0.707>}
 spine_01 {parent=pelvis, loc=<6.7494 -0.6388 0>, rot=<0 0 0.034 -0.9994>}
 spine_02 {parent=spine_01, loc=<7.2518 -0.035 0>, rot=<0 0 0.0238 -0.9997>}
 spine_03 {parent=spine_02, loc=<6.0104 0 0>, rot=<0 0 0.0149 -0.9999>}
 spine_04 {parent=spine_03, loc=<8.1024 0 0>, rot=<0 0 -0.0132 -0.9999>}
 spine_05 {parent=spine_04, loc=<10.0146 0 0>, rot=<0 0 -0.0015 -1>}
 neck_01 {parent=spine_05, loc=<11.3138 0.1529 0>, rot=<0 0 -0.2003 -0.9797>}
 neck_02 {parent=neck_01, loc=<4.5627 0 0>, rot=<0 0 0.0504 -0.9987>}
 head {parent=neck_02, loc=<5.3914 0 0>, rot=<0 0 0.115 -0.9934>}
 def_hood_root_MSTR {parent=head, loc=<-7.9925 -1.91 0>, rot=<-0.7022 -0.0827 -0.0827 -0.7022>}
 def_hood_root {parent=def_hood_root_MSTR, loc=<0 0 0>, rot=<0 0 0 1>}
 def_hood_bk_MSTR {parent=def_hood_root, loc=<16.7095 -0.0675 10.1695>, rot=<-0.003 0 1 0.0013>}
 def_hood_bk {parent=def_hood_bk_MSTR, loc=<0 0 0>, rot=<0 0 0 1>}
 def_hood_r_MSTR {parent=def_hood_root, loc=<17.7662 6.8487 -4.0286>, rot=<-0.156 -0.7536 -0.6278 -0.1168>}
 def_hood_r {parent=def_hood_r_MSTR, loc=<0 0 0>, rot=<0 0 0 1>}
 def_hood_bk_r_MSTR {parent=def_hood_root, loc=<17.2617 7.4437 4.8568>, rot=<0.0482 0.4665 0.8762 0.1112>}
 def_hood_bk_r {parent=def_hood_bk_r_MSTR, loc=<0 0 0>, rot=<0 0 0 1>}
 def_hood_l_MSTR {parent=def_hood_root, loc=<17.7658 -6.8487 -4.0287>, rot=<0.156 -0.7536 0.6278 -0.1168>}
 def_hood_l {parent=def_hood_l_MSTR, loc=<0 0 0>, rot=<0 0 0 1>}
 def_hood_bk_l_MSTR {parent=def_hood_root, loc=<17.014 -6.435 4.8108>, rot=<0.0522 -0.4985 0.8584 -0.1093>}
 def_hood_bk_l {parent=def_hood_bk_l_MSTR, loc=<0 0 0>, rot=<0 0 0 1>}
 anim_hood_r {parent=def_hood_root, loc=<17.7232 6.72 -4.9052>, rot=<0.596 0.4699 0.505 -0.4109>}
 anim_hood_top_02 {parent=def_hood_root, loc=<21.2466 0 7.2541>, rot=<0.4219 0 0.9066 0>}
 anim_hood_top_01 {parent=def_hood_root, loc=<22.3208 0 -6.7382>, rot=<0.7879 0 0.6157 0>}
 anim_hood_l {parent=def_hood_root, loc=<17.7234 -6.72 -4.9052>, rot=<0.596 -0.4699 0.505 0.4109>}
 clavicle_l {parent=spine_05, loc=<4.3575 -0.1466 -2.3223>, rot=<0.0465 -0.7289 -0.009 -0.6829>}
 upperarm_l {parent=clavicle_l, loc=<15.5111 0 0>, rot=<0.0226 -0.3234 -0.0088 -0.9459>}
 upperarm_twist_01_l {parent=upperarm_l, loc=<0.0041 0 0>, rot=<0 0 0 1>}
 clavicle_r {parent=spine_05, loc=<4.3575 -0.1466 2.3223>, rot=<0.7289 0.0465 -0.6829 0.009>}
 upperarm_r {parent=clavicle_r, loc=<-15.5111 0 0>, rot=<0.0226 -0.3234 -0.0088 -0.9459>}
 upperarm_twist_01_r {parent=upperarm_r, loc=<-0.0041 0 0>, rot=<0 0 0 -1>}
 dyn_hoodTraps_l_MSTR {parent=spine_05, loc=<7.0579 7.3538 -12.1461>, rot=<0.1408 0.693 0.6944 0.1336>}
 dyn_hoodTraps_01_l {parent=dyn_hoodTraps_l_MSTR, loc=<0 0 0>, rot=<0 0 0 1>}
 dyn_hoodTraps_02_l {parent=dyn_hoodTraps_01_l, loc=<9.1889 0 0>, rot=<-0.0008 -0.1213 0.0043 -0.9926>}
 dyn_hoodTraps_r_MSTR {parent=spine_05, loc=<6.8453 7.3317 13.1276>, rot=<-0.1014 -0.7 -0.6983 -0.1097>}
 dyn_hoodTraps_01_r {parent=dyn_hoodTraps_r_MSTR, loc=<0 0 0>, rot=<0 0 0 1>}
 dyn_hoodTraps_02_r {parent=dyn_hoodTraps_01_r, loc=<10.1279 0 0>, rot=<0.0025 -0.0982 0.0058 -0.9952>}
 MaterialCurveB {type=Curve, loc=<0 0 0>, rot=<0 0 0 1>}
]
