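# Physics asset for the dynamic hood face accessory on the male medium HeavyRoar player body.
# Capsules on spine_05 and both upper arms form body volumes alongside five hood bones and two-link left and right hoodTraps chains.
# Small MSTR capsules join spine_05 with free angular motion and anchor each hood piece.
# Hood bones have tight swing and twist limits, while hoodTraps links swing widely with minimal twist.
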PhysicsAsset /Game/Characters/Player/Male/Medium/Bodies/M_MED_HeavyRoar/Meshes/Parts/M_MED_HeavyRoar_FaceAcc_Dyn_Hood
bodies {
 def_hood_bk [{shapeType=Capsule, length=5, radius=5, shapeName=def_hood_bk_l_capsule, center=<1.5 0 -0.5>, rotation=<rot 0 0 90>}]
 def_hood_bk_MSTR [{
   shapeType=Capsule
   length=0.7039
   radius=0.6317
   shapeName=dyn_hoodTraps_l_MSTR_capsule
   center=<-0.3482 0.1201 -0.1083>
   rotation=<rot 0 -87.9178 -80.7651>
  }]
 def_hood_bk_l [{shapeType=Capsule, length=5, radius=5, shapeName=def_hood_bk_l_capsule, center=<1.5 0 -0.5>, rotation=<rot 0 0 90>}]
 def_hood_bk_l_MSTR [{
   shapeType=Capsule
   length=0.7039
   radius=0.6317
   shapeName=dyn_hoodTraps_l_MSTR_capsule
   center=<-0.3482 0.1201 -0.1083>
   rotation=<rot 0 -87.9178 -80.7651>
  }]
 def_hood_bk_r [{shapeType=Capsule, length=5, radius=5, shapeName=def_hood_bk_r_capsule, center=<1.5 0 -0.5>, rotation=<rot 0 0 90>}]
 def_hood_bk_r_MSTR [{
   shapeType=Capsule
   length=0.7039
   radius=0.6317
   shapeName=dyn_hoodTraps_l_MSTR_capsule
   center=<-0.3482 0.1201 -0.1083>
   rotation=<rot 0 -87.9178 -80.7651>
  }]
 def_hood_l [{shapeType=Capsule, length=5, radius=5, shapeName=def_hood_l_capsule, center=<1.5 0 -0.5>, rotation=<rot 0 0 90>}]
 def_hood_l_MSTR [{
   shapeType=Capsule
   length=0.7039
   radius=0.6317
   shapeName=dyn_hoodTraps_l_MSTR_capsule
   center=<-0.3482 0.1201 -0.1083>
   rotation=<rot 0 -87.9178 -80.7651>
  }]
 def_hood_r [{shapeType=Capsule, length=5, radius=5, shapeName=def_hood_r_capsule, center=<1.5 0 -0.5>, rotation=<rot 0 0 90>}]
 def_hood_r_MSTR [{
   shapeType=Capsule
   length=0.7039
   radius=0.6317
   shapeName=dyn_hoodTraps_l_MSTR_capsule
   center=<-0.3482 0.1201 -0.1083>
   rotation=<rot 0 -87.9178 -80.7651>
  }]
 dyn_hoodTraps_01_l [{shapeType=Capsule, length=4.75, radius=2, shapeName=dyn_hoodTraps_01_l_capsule, center=<6 0 0.5>, rotation=<rot 90 0 0>}]
 dyn_hoodTraps_01_r [{shapeType=Capsule, length=4.25, radius=2, shapeName=dyn_hoodTraps_01_r_capsule, center=<6.5843 0 0.5>, rotation=<rot 90 0 0>}]
 dyn_hoodTraps_02_l [{shapeType=Capsule, length=6.5, radius=2, shapeName=dyn_hoodTraps_02_l_capsule, center=<6 0 0>, rotation=<rot 90 0 0>}]
 dyn_hoodTraps_02_r [{shapeType=Capsule, length=4.75, radius=2, shapeName=dyn_hoodTraps_02_r_capsule, center=<6 0 0.5>, rotation=<rot 90 0 0>}]
 dyn_hoodTraps_l_MSTR [{
   shapeType=Capsule
   length=0.7039
   radius=0.6317
   shapeName=dyn_hoodTraps_l_MSTR_capsule
   center=<-0.3482 0.1201 -0.1083>
   rotation=<rot 0 -87.9178 -80.7651>
  }]
 dyn_hoodTraps_r_MSTR [{
   shapeType=Capsule
   length=0.505
   radius=0.505
   shapeName=dyn_hoodTraps_r_MSTR_capsule
   center=<-0.2483 -0.0052 -0.2903>
   rotation=<rot 0 68.15 -64.6069>
  }]
 spine_05 [
  {shapeType=Capsule, length=17.4761, radius=14.3905, shapeName=spine_05_capsule, center=<-4.7659 -4.5669 0>}
  {shapeType=Capsule, length=3, radius=15, shapeName=spine_05_capsule, center=<-3.6899 -2.262 0>}
 ]
 upperarm_l [{shapeType=Capsule, length=21.25, radius=6.5, shapeName=upperarm_r_capsule, center=<11.4176 -2.5885 0.6771>, rotation=<rot 90 180 -9.1689>}]
 upperarm_r [{
   shapeType=Capsule
   length=17.75
   radius=6
   shapeName=upperarm_r_capsule
   center=<-11.4176 2.5885 -0.6771>
   rotation=<rot 90 -3.5763 -12.7453>
  }]
}
constraints [
 {bone1Name=dyn_hoodTraps_01_l, bone2Name=dyn_hoodTraps_l_MSTR, swing1LimitDegrees=60, swing2LimitDegrees=45, twistLimitDegrees=2}
 {bone1Name=dyn_hoodTraps_02_l, bone2Name=dyn_hoodTraps_01_l, swing1LimitDegrees=60, swing2LimitDegrees=45, twistLimitDegrees=2}
 {bone1Name=dyn_hoodTraps_01_r, bone2Name=dyn_hoodTraps_r_MSTR, swing1LimitDegrees=60, swing2LimitDegrees=45, twistLimitDegrees=2}
 {bone1Name=dyn_hoodTraps_02_r, bone2Name=dyn_hoodTraps_01_r, swing1LimitDegrees=60, swing2LimitDegrees=45, twistLimitDegrees=2}
 {
  bone1Name=dyn_hoodTraps_l_MSTR
  bone2Name=spine_05
  swing1LimitDegrees=45
  swing1Motion=Free
  swing2LimitDegrees=45
  swing2Motion=Free
  twistLimitDegrees=45
  twistMotion=Free
 }
 {
  bone1Name=dyn_hoodTraps_r_MSTR
  bone2Name=spine_05
  swing1LimitDegrees=45
  swing1Motion=Free
  swing2LimitDegrees=45
  swing2Motion=Free
  twistLimitDegrees=45
  twistMotion=Free
 }
 {
  bone1Name=upperarm_r
  bone2Name=spine_05
  swing1LimitDegrees=45
  swing1Motion=Free
  swing2LimitDegrees=45
  swing2Motion=Free
  twistLimitDegrees=45
  twistMotion=Free
 }
 {
  bone1Name=upperarm_l
  bone2Name=spine_05
  swing1LimitDegrees=45
  swing1Motion=Free
  swing2LimitDegrees=45
  swing2Motion=Free
  twistLimitDegrees=45
  twistMotion=Free
 }
 {
  bone1Name=def_hood_bk_l_MSTR
  bone2Name=spine_05
  swing1LimitDegrees=45
  swing1Motion=Free
  swing2LimitDegrees=45
  swing2Motion=Free
  twistLimitDegrees=45
  twistMotion=Free
 }
 {bone1Name=def_hood_bk_l, bone2Name=def_hood_bk_l_MSTR, swing1LimitDegrees=5, swing2LimitDegrees=10, twistLimitDegrees=2}
 {
  bone1Name=def_hood_bk_r_MSTR
  bone2Name=spine_05
  swing1LimitDegrees=45
  swing1Motion=Free
  swing2LimitDegrees=45
  swing2Motion=Free
  twistLimitDegrees=45
  twistMotion=Free
 }
 {bone1Name=def_hood_bk_r, bone2Name=def_hood_bk_r_MSTR, swing1LimitDegrees=5, swing2LimitDegrees=10, twistLimitDegrees=2}
 {
  bone1Name=def_hood_l_MSTR
  bone2Name=spine_05
  swing1LimitDegrees=45
  swing1Motion=Free
  swing2LimitDegrees=45
  swing2Motion=Free
  twistLimitDegrees=45
  twistMotion=Free
 }
 {bone1Name=def_hood_l, bone2Name=def_hood_l_MSTR, swing1LimitDegrees=5, swing2LimitDegrees=10, twistLimitDegrees=2}
 {
  bone1Name=def_hood_r_MSTR
  bone2Name=spine_05
  swing1LimitDegrees=45
  swing1Motion=Free
  swing2LimitDegrees=45
  swing2Motion=Free
  twistLimitDegrees=45
  twistMotion=Free
 }
 {bone1Name=def_hood_r, bone2Name=def_hood_r_MSTR, swing1LimitDegrees=5, swing2LimitDegrees=10, twistLimitDegrees=2}
 {
  bone1Name=def_hood_bk_MSTR
  bone2Name=spine_05
  swing1LimitDegrees=45
  swing1Motion=Free
  swing2LimitDegrees=45
  swing2Motion=Free
  twistLimitDegrees=45
  twistMotion=Free
 }
 {bone1Name=def_hood_bk, bone2Name=def_hood_bk_MSTR, swing1LimitDegrees=5, swing2LimitDegrees=10, twistLimitDegrees=2}
]
