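# Animation Blueprint for the HeavyRoar face accessory, posing it from the character's meshes.
# Output copies the attached parent's pose, layering left front jacket bones from BodyMesh, which is cached when customization loads.
# A hood Control Rig feeds spine_05-space RigidBody physics, but that chain is not connected to the output.
# Each update inverts material parameter "Hide Element 02" into isWearingHood, the physics alpha.
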AnimBlueprint /BRCosmetics/Characters/Player/Male/Medium/Bodies/M_MED_HeavyRoar/Meshes/Parts/M_MED_HeavyRoar_FaceAcc_AnimBP
target_skeleton=&/M_MED_HeavyRoar_FaceAcc_Skeleton
graphs [
 AnimGraph {nodes [
   n1 = Control Rig() {
    alphaBoolBlend {blendOption=Linear}
    controlRigAssetReference {blueprintRigClass=&/M_MED_HeavyRoar_Hood_CtrlRig.M_MED_HeavyRoar_Hood_CtrlRig_C}
   }
   n2 = RigidBody!(
    ComponentPose=Local To Component(LocalPose=@n1)
    Alpha=Get isWearingHood() -> (isWearingHood:double)
    OverrideWorldGravity:Vector="0,0,-980"
    ExternalForce
   ) -> (Pose) {
    bClampLinearTranslationLimitToRefPose=true
    bOverrideWorldGravity=true
    overridePhysicsAsset=/Game/Characters/Player/Male/Medium/Bodies/M_MED_HeavyRoar/Meshes/Parts/M_MED_HeavyRoar_FaceAcc_Dyn_Hood
    simulationSpace=BaseBoneSpace
    alphaBoolBlend {blendOption=Linear}
    baseBoneRef {boneName=spine_05}
    componentAppliedLinearAccClamp=<3000 1000 1000>
    componentLinearAccScale {x=0.15}
    componentLinearVelScale {x=0.2}
    overrideWorldGravity {z=-980}
    simSpaceSettings {maxAngularAcceleration=100, maxAngularVelocity=2, maxLinearAcceleration=500, maxLinearVelocity=100, velocityScaleZ=0, worldAlpha=1}
   }
   Component To Local(ComponentPose=@n2)
   n3 = Layered blend per bone(
    BasePose=Copy Pose From Mesh() {bUseAttachedParent=true}
    BlendPoses_0=Copy Pose From Mesh(SourceMeshComponent=Get BodyMesh() -> (BodyMesh:Skeletal Mesh Component Object Reference)) {bUseAttachedParent=true}
   ) {layerSetup [{branchFilters [{boneName=dyn_jacket_l_fr_MSTR}, {boneName=dyn_jacket_l_fr_01}, {boneName=dyn_jacket_l_fr_02}, {boneName=dyn_jacket_l_fr_03}]}]}
   Output Pose(Result=@n3) {layerGroup=None}
   Slot 'DefaultSlot'()
  ]}
 EventGraph {nodes [
   Event BlueprintUpdateAnimation()
   > n1 = Get Dynamic Material Instance(SkeletalMeshComponent=GetPartSkeletalMeshComponent())
   > n2 = Get Scalar Parameter Value!(execute, self:Material Instance Dynamic Object Reference=@n1, ParameterName:Name=Hide Element 02) -> (then, ReturnValue:float) {is=title K2Node_CallFunction}
   > Set isWearingHood(isWearingHood:double=MapRangeClamped(Value=@n2, InRangeA=0, InRangeB=1, OutRangeA=1, OutRangeB=0)) -> (Output_Get:double)
   Event OnCharacterCustomizationLoaded()
   > n3 = Cast To PlayerPawn_Athena(Object=TryGetPawnOwner())
   > Set BodyMesh(BodyMesh=GetSkeletalMeshForPartType(self=@n3))
  ]}
]
variables [
 BodyMesh {type=SkeletalMeshComponent}
 isWearingHood {type=number}
]
